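# Niagara system for the Snowfall emote's falling confetti effect on the main player.
# Three emitters (Confetti, RibbonConfetti, Sparkles) spawn continuously in cylinder volumes, fall under drag and randomized drifting gravity, then shrink by curve.
# All use sprite renderers, including RibbonConfetti; Sparkles are orange HDR and twinkle through a jagged size curve.
# Burst spawns are disabled, and the exposed SkeletalMesh and RibbonIDOffset user parameters are unreferenced by the listed modules.
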NiagaraSystem /BRCosmetics/Animation/Game/MainPlayer/Emotes/Snowfall/FX/NS_Emote_Snowfall_Confetti
dependencies {
 usedDataInterfaces [&NiagaraDataInterfaceCurve, &NiagaraDataInterfaceColorCurve]
 usedDynamicInputs [&RotateVectorByQuaternion]
 usedModules [
  &SystemState
  &self:SystemSpawnScript.SystemScriptSource.SystemScriptGraph.NiagaraNodeAssignment_0.SetVariables_#1
 ]
}
emitters {
 Confetti {
  emitterUpdateScript [
   &EmitterState()
   &SpawnBurst_Instantaneous(Spawn Count=35, Use Spawn Probability=false) {enabled=false}
   &SpawnRate(SpawnRate=12, Use Spawn Probability=false)
  ]
  particleSpawnScript [
   /Niagara/Modules/Spawn/Initialization/InitializeParticle(
    Color=&MakeLinearColorFromVectorAndFloat(
     "Vector (RGB)"=&MakeVector(X=&RandomRangeFloat(Minimum=0, Maximum=-0.05), Y=0, Z=0)
     "Float (Alpha)"=1
    )
    Lifetime=&RandomRangeFloat(Minimum=1, Maximum=1.85)
    Module.Write Color=true
    Module.Write Lifetime=true
    Module.Write Mass=true
    Module.Write Material Random=true
    Module.Write Position=false
    Module.Write Scale=false
    Module.Write SpriteRotation=true
    Module.Write SpriteSize=true
    Position=&SimulationPosition()
    Sprite Rotation=&RandomRangeFloat(Minimum=0, Maximum=360)
    Sprite Size=&Vector2DFromFloat(Value=&RandomRangeFloat(Minimum=2, Maximum=5))
   )
   &ShapeLocation(
    Cylinder Height=20
    Cylinder Radius=125
    MaskSpawnGroup=false
    Non Uniform Scale=<1 1 0.5>
    Override Seed=false
    Shape Primitive=::NewEnumerator1
    Surface Only=true
    UseOwnerScale=true
   )
   &InheritVelocity(Inherited Velocity Amount Scale=<0 0 -1>, Inherited Velocity Speed Limit=-5, Speed Limit=true, UseSpeedThreshold=false)
   &AddVelocity(Velocity=$System.DirectionalVelocity) {enabled=false}
  ]
  particleUpdateScript [
   &ParticleState()
   &Drag(Drag=6, Use Linear Drag=true, Use Rotational Drag=false)
   &GravityForce(Gravity=&RandomRangeVector(Minimum=<50 60 -360>, Maximum=<100 150 -200>))
   &SolveForcesAndVelocity(Clamp Velocity=false, Limit Acceleration=false)
   &ScaleSpriteSize(Scale Factor=&Vector2DFromFloat(Value=&FloatFromCurve(FloatCurve=DataInterface() {propertyValues {Curve {keys [
          {arriveTangent=-0.4596, interpMode=RCIM_Cubic, leaveTangent=-0.4596, tangentMode=RCTM_User, time=0.55, value=1}
          {arriveTangent=-4.9031, interpMode=RCIM_Cubic, leaveTangent=-4.9031, tangentMode=RCTM_User, time=1, value=0}
         ]}}})))
   &ScaleColor(
    Scale RGB=&MakeVector(
     X=1
     Y=&Add_Float(
      A=&RandomRangeFloat(Minimum=0, Maximum=1000)
      B=&Lerp_Float(A=0, B=&RandomRangeFloat(Minimum=25, Maximum=150), Alpha=$Particles.NormalizedAge)
     )
     Z=1
    )
    ScaleA=false
    ScaleRGB=true
   )
   /Niagara/Modules/Update/Color/ScaleColorBySpeed() {enabled=false}
  ]
  rendererClasses [&NiagaraSpriteRendererProperties]
  renderers [{rendererClass=&NiagaraSpriteRendererProperties}]
 }
 RibbonConfetti {
  emitterUpdateScript [
   &EmitterState()
   &SpawnBurst_Instantaneous(Spawn Count=5, Use Spawn Probability=false) {enabled=false}
   &SpawnRate(SpawnRate=15, Use Spawn Probability=false)
  ]
  particleSpawnScript [
   &InitializeParticle(
    Color=&MakeLinearColorFromVectorAndFloat(
     "Vector (RGB)"=&MakeVector(X=&RandomRangeFloat(Minimum=0, Maximum=0.05), Y=0, Z=0)
     "Float (Alpha)"=1
    )
    Lifetime=&RandomRangeFloat(Minimum=1, Maximum=1.95)
    Module.Write Color=true
    Module.Write Lifetime=true
    Module.Write Mass=true
    Module.Write Material Random=true
    Module.Write Position=false
    Module.Write Scale=false
    Module.Write SpriteRotation=true
    Module.Write SpriteSize=true
    Position=&SimulationPosition()
    Sprite Rotation=&RandomRangeFloat(Minimum=0, Maximum=360)
    Sprite Size=&Vector2DFromFloat(Value=&RandomRangeFloat(Minimum=5, Maximum=10))
   )
   &ShapeLocation(
    Cylinder Height=25
    Cylinder Radius=145
    MaskSpawnGroup=false
    Non Uniform Scale=<1 1 0.5>
    Override Seed=false
    Shape Primitive=::NewEnumerator1
    Surface Only=false
    UseOwnerScale=true
   ) {module=ShapeLocation001}
   &ShapeLocation(
    Cylinder Height=25
    Cylinder Radius=50
    MaskSpawnGroup=false
    Override Seed=false
    Shape Primitive=::NewEnumerator1
    Surface Only=false
    UseOwnerScale=true
   ) {enabled=false}
   &InheritVelocity(Inherited Velocity Amount Scale=<0 0 -1>, Inherited Velocity Speed Limit=-5, Speed Limit=true, UseSpeedThreshold=false)
   &AddVelocityFromPoint(Module.Use Falloff Distance=false) {enabled=false}
  ]
  particleUpdateScript [
   &ParticleState()
   &Drag(Drag=5, Use Linear Drag=true, Use Rotational Drag=false)
   &GravityForce(Gravity=&RandomRangeVector(Minimum=<35 60 -360>, Maximum=<75 150 -200>))
   &SolveForcesAndVelocity(Clamp Velocity=false, Limit Acceleration=false)
   &ScaleSpriteSize(Scale Factor=&Vector2DFromFloat(Value=&FloatFromCurve(FloatCurve=DataInterface() {propertyValues {Curve {keys [
          {arriveTangent=-0.4596, interpMode=RCIM_Cubic, leaveTangent=-0.4596, tangentMode=RCTM_User, time=0.55, value=1}
          {arriveTangent=-4.0966, interpMode=RCIM_Cubic, leaveTangent=-4.0966, tangentMode=RCTM_User, time=1, value=0}
         ]}}})))
  ]
  rendererClasses [&NiagaraSpriteRendererProperties]
  renderers [{rendererClass=&NiagaraSpriteRendererProperties}]
 }
 Sparkles {
  emitterUpdateScript [
   &EmitterState()
   &SpawnBurst_Instantaneous(Spawn Count=30, Use Spawn Probability=false) {enabled=false}
   &SpawnRate(SpawnRate=13, Use Spawn Probability=false)
  ]
  particleSpawnScript [
   &InitializeParticle(
    Color=<rgba 5 1.6469 0 1>
    Lifetime=&RandomRangeFloat(Minimum=1, Maximum=1.9)
    Module.Write Color=true
    Module.Write Lifetime=true
    Module.Write Mass=true
    Module.Write Material Random=true
    Module.Write Position=false
    Module.Write Scale=false
    Module.Write SpriteRotation=true
    Module.Write SpriteSize=true
    Position=&SimulationPosition()
    Sprite Rotation=&RandomRangeFloat(Minimum=0, Maximum=360)
    Sprite Size=&RandomRangeVector2D(Minimum=<2 2>, Maximum=<6 6>)
   )
   &ShapeLocation(
    Cylinder Height=20
    Cylinder Radius=125
    MaskSpawnGroup=false
    Non Uniform Scale=<1 1 0.5>
    Override Seed=false
    Shape Primitive=::NewEnumerator1
    Surface Only=false
    UseOwnerScale=true
   ) {module=ShapeLocation001}
   &ShapeLocation(
    Cylinder Height=20
    Cylinder Radius=65
    MaskSpawnGroup=false
    Override Seed=false
    Shape Primitive=::NewEnumerator1
    Surface Only=false
    UseOwnerScale=true
   ) {enabled=false}
   &InheritVelocity(Inherited Velocity Amount Scale=<0 0 -1>, Inherited Velocity Speed Limit=-5, Speed Limit=true, UseSpeedThreshold=false)
   &AddVelocityFromPoint(Module.Use Falloff Distance=false) {enabled=false}
  ]
  particleUpdateScript [
   &ParticleState()
   &Drag(Drag=3, Use Linear Drag=true, Use Rotational Drag=false)
   &GravityForce(Gravity=&RandomRangeVector(Minimum=<35 65 -360>, Maximum=<75 150 -300>))
   &SolveForcesAndVelocity(Clamp Velocity=false, Limit Acceleration=false)
   &ScaleColor(Scale Alpha=&Lerp_Float(A=1, B=0, Alpha=$Particles.NormalizedAge), ScaleA=true, ScaleRGB=false) {enabled=false}
   &ScaleColor(
    Scale RGB=&MakeVectorFromLinearColorRGB(Color=&ColorFromCurve(ColorCurve=DataInterface() {propertyValues {
        AlphaCurve {keys [<at 0 0 RCIM_Cubic>,<at 0.052 1>,<at 0.1435 0.1>,<at 0.2605 1>,<at 0.3836 0.1>,<at 0.4918 1>,<at 0.6968 0.1677>,<at 1 0 RCIM_Cubic>]}
        BlueCurve {keys [<at 0 16.6566 RCIM_Cubic>,<at 0.361 3.2515>,<at 0.4146 0.6269>,<at 1 0 RCIM_Cubic>]}
        GreenCurve {keys [<at 0 0 RCIM_Cubic>,<at 0.361 1.4098>,<at 0.4146 0.2926>,<at 1 0.4788 RCIM_Cubic>]}
        RedCurve {keys [<at 0 20 RCIM_Cubic>,<at 0.361 5>,<at 0.4146 0.9906>,<at 1 1 RCIM_Cubic>]}
       }}))
    ScaleA=true
    ScaleRGB=true
   ) {enabled=false, module=ScaleColor001}
   &ScaleSpriteSize(Scale Factor=&Vector2DFromFloat(Value=&FloatFromCurve(FloatCurve=DataInterface() {propertyValues {Curve {keys [<at 0.0018 0.1957>,<at 0.0527 0.819>,<at 0.1169 0.6555>,<at 0.1748 0.9015>,<at 0.2394 0.582>,<at 0.332 1.0023>,<at 0.4947 0.3707>,<at 0.6727 1>,<at 1 0>]}}})))
  ]
  rendererClasses [&NiagaraSpriteRendererProperties]
  renderers [{rendererClass=&NiagaraSpriteRendererProperties}]
 }
}
summary {
 systemName=NS_Emote_Snowfall_Confetti
 emitters [
  {emitterName=Confetti, rendererClasses [&NiagaraSpriteRendererProperties]}
  {emitterName=RibbonConfetti, rendererClasses [&NiagaraSpriteRendererProperties]}
  {emitterName=Sparkles, rendererClasses [&NiagaraSpriteRendererProperties]}
 ]
 userVariables [
  User.RibbonIDOffset {type=NiagaraInt32}
  User.SkeletalMesh {
   type=NiagaraDataInterfaceSkeletalMesh
   defaultValue {
    struct=NiagaraExt_VariableValue_DataInterface
    value {dataInterface=&self:NiagaraDataInterfaceSkeletalMesh_0}
   }
  }
 ]
}
