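# Niagara system for the snow impact effect of the Snowfall emote, combining a puff of snow smoke with snow meshes around the board.
# Smoke001 bursts short-lived sprites in an upward cone that gravity pulls down.
# SnowAroundBoard bursts longer-lived GPU mesh particles, whose custom curves pop their size and height up, then shrink and sink them to fake melting.
# A front-end modifier slightly dims the mesh color outside gameplay.
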NiagaraSystem /BRCosmetics/Animation/Game/MainPlayer/Emotes/Snowfall/FX/NS_Emote_Snowfall_ImpactSnow
dependencies {
 usedDataInterfaces [&NiagaraDataInterfaceCurve, &NiagaraDataInterfaceSkeletalMesh]
 usedDynamicInputs []
 usedModules [&SampleSkeletalMesh, &SystemState]
}
emitters {
 Smoke001 {
  emitterUpdateScript [
   &EmitterState(
    Inactive Response=::NewEnumerator2
    Life Cycle Mode=::NewEnumerator1
    Loop Behavior=::NewEnumerator2
    Loop Count=2
    Loop Delay=0.95
    Loop Duration=0
    UseLoopDelay=false
   )
   &SpawnBurst_Instantaneous(Spawn Count=9, Use Spawn Probability=false, UseLoopCountLimit=false)
   &SpawnRate(
    SpawnRate=&FloatFromCurve(
     FloatCurve=DataInterface() {propertyValues {Curve {keys [{interpMode=RCIM_Cubic, tangentMode=RCTM_User, time=0, value=1}, {interpMode=RCIM_Cubic, tangentMode=RCTM_User, time=1, value=0}]}}}
     CurveIndex=$Emitter.Age
     Scale Curve=25
    )
    Use Spawn Probability=false
   ) {enabled=false}
  ]
  particleSpawnScript [
   &InitializeParticle(
    Color=<rgba 0.4492 0.6108 0.8177 0.2>
    Lifetime Max=0.6
    Lifetime Min=0.3
    Lifetime Mode=::NewEnumerator1
    Sprite Rotation Mode=::NewEnumerator2
    Sprite Size Mode=::NewEnumerator1
    Uniform Sprite Size Max=50
    Uniform Sprite Size Min=25
    UsePositionOffset=false
    Write Material Random=true
   )
   &ShapeLocation(
    Cylinder Height=5
    Cylinder Height Midpoint=0
    Cylinder Radius=10
    MaskSpawnGroup=false
    Non Uniform Scale=<1 1 0.5>
    Offset=<10 -45 -10>
    Offset Coordinate Space=::Local
    Offset Mode=::NewEnumerator0
    Override Seed=false
    Shape Primitive=::NewEnumerator1
    Surface Only=false
    UseOwnerScale=true
   )
   &CameraOffset()
   &DynamicMaterialParameters(
    Index 0 Param 1=1
    Index 0 Param 2=0.5
    Index 0 Param 3=0.5
    Module.Param0WriteEnabled=true
    Module.Param1WriteEnabled=true
    Module.Param2WriteEnabled=true
    Module.Param3WriteEnabled=true
   )
   &InheritVelocity(Inherited Velocity Speed Limit=200, Speed Limit=true, UseSpeedThreshold=false) {enabled=false}
   &AddVelocity(
    Cone Angle=40
    Cone Axis=<0 0 -1>
    Module.UseConeFalloff=false
    Module.UseVelDistribution=false
    Override Seed=false
    Velocity Mode=::NewEnumerator2
    Velocity Speed=-175
   )
  ]
  particleUpdateScript [
   &GravityForce(Gravity=&RandomRangeVector(Minimum=<0 0 -500>, Maximum=<0 0 -300>))
   &SolveForcesAndVelocity(Clamp Velocity=false, Limit Acceleration=false)
   &ParticleState()
   &ScaleColor(
    Scale Alpha=&FloatFromCurve(
     FloatCurve=DataInterface() {propertyValues {Curve {keys [{interpMode=RCIM_Cubic, tangentMode=RCTM_User, time=0.85, value=1}, {interpMode=RCIM_Cubic, tangentMode=RCTM_User, time=2, value=0}]}}}
     CurveIndex=$Emitter.Age
    )
    ScaleA=true
    ScaleRGB=true
   )
   &SubUVAnimation()
   &ScaleSpriteSize(Uniform Curve Sprite Scale=DataInterface() {propertyValues {Curve {keys [
        {interpMode=RCIM_Cubic, tangentMode=RCTM_User, time=0, value=0}
        {interpMode=RCIM_Cubic, tangentMode=RCTM_User, time=0.9, value=1.125}
        {tangentMode=RCTM_User, time=1, value=0}
       ]}}})
  ]
  rendererClasses [&NiagaraSpriteRendererProperties]
  renderers [{rendererClass=&NiagaraSpriteRendererProperties}]
 }
 SnowAroundBoard {
  simTarget=GPUComputeSim
  emitterUpdateScript [
   &EmitterState(
    Delay First Loop Only=true
    Inactive Response=::NewEnumerator2
    Life Cycle Mode=::NewEnumerator1
    Loop Behavior=::NewEnumerator2
    Loop Count=2
    Loop Delay=0.05
    Loop Duration=0
    UseLoopDelay=true
   ) {module=EmitterState001}
   &SpawnBurst_Instantaneous(Loop Count Limit=0, Spawn Count=9, Use Spawn Probability=false, UseLoopCountLimit=false)
  ]
  particleSpawnScript [
   &InitializeParticle(
    Color=&MultiplyLinearColor(
     Linear Color=<rgba 0.6302 0.8174 1 1>
     Scale Factor=&MakeVector4_FromFloat(Float=&FN_FrontEndFloatModifier(Gameplay=1, Front End=0.8))
    )
    Lifetime=&RandomRangeFloat(Minimum=3.2, Maximum=3.5)
    Mesh Scale=&Multiply_VectorByFloat(
     Vector=&RandomRangeVector(Minimum=<0.75 0.5 0.5>, Maximum=<1.1 0.75 0.75>)
     Float=&RandomRangeFloat(Minimum=0.08, Maximum=0.03)
    )
    Mesh Scale Mode=::NewEnumerator4
    Sprite Rotation Mode=::NewEnumerator2
    UsePositionOffset=false
    Write Material Random=true
   )
   &ShapeLocation(
    Cylinder Height=5
    Cylinder Radius=15
    MaskSpawnGroup=false
    Non Uniform Scale=<1.1 0.675 0.5>
    Offset=<10 -44 -45>
    Offset Coordinate Space=::Local
    Offset Mode=::NewEnumerator0
    Override Seed=false
    Shape Primitive=::NewEnumerator1
    Surface Only=false
    UseOwnerScale=true
   )
   &InitialMeshOrientation(Rotation=&RandomRangeVector(Minimum=<0 0 0>, Maximum=<1 1 1>), Use Rotation Vector=true) {module=InitialMeshOrientation001}
   &AddVelocity(Velocity=<0 0 750>) {enabled=false}
  ]
  particleUpdateScript [
   /SetVariables(
    Particles.AnimMeltAway:NiagaraFloat=&FloatFromCurve(FloatCurve=DataInterface() {propertyValues {Curve {keys [
         {interpMode=RCIM_Cubic, tangentMode=RCTM_User, time=0, value=0.8}
         {interpMode=RCIM_Cubic, tangentMode=RCTM_User, time=0.05, value=1}
         {arriveTangent=-0.039, interpMode=RCIM_Cubic, leaveTangent=-0.039, tangentMode=RCTM_User, time=0.85, value=1}
         {tangentMode=RCTM_User, time=1, value=0}
        ]}}})
    Particles.AnimSize:NiagaraFloat=&FloatFromCurve(FloatCurve=DataInterface() {propertyValues {Curve {keys [
         {interpMode=RCIM_Cubic, leaveTangentWeight=0.01, tangentMode=RCTM_User, time=0, value=0.75}
         {arriveTangentWeight=0.01, interpMode=RCIM_Cubic, leaveTangentWeight=0.29, tangentMode=RCTM_User, time=0.05, value=1}
         {
          arriveTangent=0.0156
          arriveTangentWeight=0.2901
          interpMode=RCIM_Cubic
          leaveTangent=0.0156
          leaveTangentWeight=0.0333
          tangentMode=RCTM_User
          time=0.9
          value=1
         }
         <at 1 0>
        ]}}})
   ) {isSetParameters=true}
   &Drag(Drag=55, Use Linear Drag=true, Use Rotational Drag=false) {enabled=false}
   &SolveForcesAndVelocity(Clamp Velocity=false, Limit Acceleration=false)
   &ParticleState()
   &ScaleColor(
    Scale Alpha=&FloatFromCurve(FloatCurve=DataInterface() {propertyValues {Curve {keys [{interpMode=RCIM_Cubic, tangentMode=RCTM_User, time=0.83, value=1}, {interpMode=RCIM_Cubic, tangentMode=RCTM_User, time=0.93, value=0}]}}})
    ScaleA=true
    ScaleRGB=true
   )
   &ScaleMeshSize(Scale Factor=&VectorFromFloat(Value=$Particles.AnimSize))
   /SetVariables(Particles.Position=&ConvertVectorToPosition(
     Input Position=&Add_Vector(
      A=&ConvertPositionToVector(
       Input Position=&ConvertVectorToPosition(Input Position=$Particles.ShapeLocation.ShapePosition, Vector Is In Absolute Large World Space:NiagaraBool=false)
       Convert To Absolute Large World Vector=false
      )
      B=&Lerp_Vector(A=<0 0 0>, B=<0 0 45>, Alpha=$Particles.AnimMeltAway)
     )
     Vector Is In Absolute Large World Space:NiagaraBool=false
    )) {isSetParameters=true}
  ]
  rendererClasses [&NiagaraMeshRendererProperties]
  renderers [{rendererClass=&NiagaraMeshRendererProperties, rendererIndex=0}]
 }
}
summary {
 systemName=NS_Emote_Snowfall_ImpactSnow
 emitters [
  {emitterName=SnowAroundBoard, simTarget=GPUComputeSim, rendererClasses [&NiagaraMeshRendererProperties]}
  {emitterName=Smoke001, rendererClasses [&NiagaraSpriteRendererProperties]}
 ]
 userVariables [
  User.bIsFrontEnd {type=NiagaraBool}
  User.bIsFrontEndPreview {type=NiagaraBool}
 ]
}
